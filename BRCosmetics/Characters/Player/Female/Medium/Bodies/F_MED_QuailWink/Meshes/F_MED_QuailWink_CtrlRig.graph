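# Control Rig for the QuailWink female medium body cosmetic, correcting deformation of its shoulder pad, chains, straps, charm and skirt.
# A shared deform-rig function runs first, then constraints, arm twist fixes, waist corrections and right-shoulder adjustments.
# Corrective transforms are weighted by deform-rig animation attributes, remapped thigh rotations and a spherical pose reader on the right upper arm.
# Crouch, emote and skydive flags switch or suppress some corrections.
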ControlRigBlueprint /BRCosmetics/Characters/Player/Female/Medium/Bodies/F_MED_QuailWink/Meshes/F_MED_QuailWink_CtrlRig
graphs [
 New Function_1_ContainedGraph {nodes [
   n1 = RigVMFunctionEntryNode()
   RigVMFunction_ControlFlowBranch = &RigVMFunction_ControlFlowBranch(Condition=RigVMVariableNode() {variable=isCrouching}.Value, ExecuteContext=@n1.ExecuteContext)
   Delta Rotation Z_1 = RigVMFunctionReferenceNode(Child=thigh_l, Parent=thigh_l_null, Source Maximum=-120, ExecuteContext=@RigVMFunction_ControlFlowBranch.Completed)
   > &RigUnit_ModifyTransforms(
    ItemToModify="((Item=(Type=Bone,Name=\"dyn_charm_01_l\"),Transform=(Rotation=(X=0.1417,Y=0.7804,Z=-0.102,W=-0.6004),Translation=(X=0,Y=0,Z=0),Scale3D=(X=1,Y=1,Z=1))),(Item=(Type=Bone,Name=\"dyn_charm_02_l\"),Transform=(Rotation=(X=0,Y=0,Z=0,W=1),Translation=(X=0,Y=0,Z=0),Sc ...(truncated)"
    WeightMinimum=0
    Weight=@"Delta Rotation Z_1".Delta Rotation
   )
   > RigVMFunctionReturnNode()
   Delta Rotation Z = RigVMFunctionReferenceNode(Child=thigh_r, Parent=thigh_r_null, Source Maximum=-120, ExecuteContext=@RigVMFunction_ControlFlowBranch.True)
   > &RigUnit_ModifyTransforms(
    ItemToModify=((Item=(Type=Bone,Name="base_side_chain_01"),Transform=(Rotation=(X=0,Y=-0.2508,Z=0,W=0.968),Translation=(X=0,Y=0,Z=0),Scale3D=(X=1,Y=1,Z=1))))
    WeightMinimum=0
    Weight=@"Delta Rotation Z".Delta Rotation
   )
   &RigUnit_ModifyTransforms(
    ItemToModify="((Item=(Type=Bone,Name=\"base_sideStrap_01_fr\"),Transform=(Rotation=(X=0,Y=0,Z=0.2161,W=0.9764),Translation=(X=0,Y=0,Z=0),Scale3D=(X=1,Y=1,Z=1))),(Item=(Type=Bone,Name=\"base_sideStrap_01_bck\"),Transform=(Rotation=(X=-0.0436,Y=-0.0084,Z=0.1673,W=0.9849),Tran ...(truncated)"
    WeightMinimum=0
    Weight=RigVMDispatchNode(BoneName=root, Default=1, Name=deformRig_r_thigh_up_alpha) {template="DISPATCH_RigDispatch_GetAnimAttribute(in Name,in BoneName,in Default,out Value,out Found)"}.Value
    ExecutePin=@RigVMFunction_ControlFlowBranch.False
   )
   > Delta Rotation Y = RigVMFunctionReferenceNode(Child=thigh_r, Parent=thigh_r_null, Source Maximum=-85)
   > n2 = &RigUnit_ModifyTransforms(
    ItemToModify=((Item=(Type=Bone,Name="base_side_chain_01"),Transform=(Rotation=(X=0,Y=-0.2114,Z=0,W=0.9774),Translation=(X=0,Y=0,Z=0),Scale3D=(X=1,Y=1,Z=1))))
    WeightMinimum=0
    Weight=@"Delta Rotation Y".Delta Rotation
   )
   n3 = &RigVMFunction_ControlFlowBranch(ExecuteContext=@n2.ExecutePin, Condition=RigVMVariableNode() {variable=isEmoting}.Value)
   &RigUnit_ModifyTransforms(
    ItemToModify="((Item=(Type=Bone,Name=\"base_bk_skirt_r\"),Transform=(Rotation=(X=0.0079,Y=-0.1857,Z=-0.1036,W=0.9771),Translation=(X=-0.1497,Y=-2.6112,Z=-0.0523),Scale3D=(X=1,Y=1,Z=1))),(Item=(Type=Bone,Name=\"base_side_skirt_r\"),Transform=(Rotation=(X=0,Y=-0.0262,Z=0,W=0. ...(truncated)"
    WeightMinimum=0
    ExecutePin=@n3.True
    Weight=@"Delta Rotation Y".Delta Rotation
   )
   &RigUnit_ModifyTransforms(
    ExecutePin=()
    ItemToModify="((Item=(Type=Bone,Name=\"base_sideStrap_01_fr\"),Transform=(Rotation=(X=0,Y=0,Z=-0.0736,W=-0.9973),Translation=(X=0,Y=0,Z=0),Scale3D=(X=1,Y=1,Z=1))),(Item=(Type=Bone,Name=\"base_sideStrap_01_bck\"),Transform=(Rotation=(X=0,Y=0,Z=0,W=1),Translation=(X=0,Y=0,Z=0 ...(truncated)"
    WeightMinimum=0
    Weight=RigVMDispatchNode(BoneName=root, Default=1, Name=deformRig_l_thigh_up_alpha) {template="DISPATCH_RigDispatch_GetAnimAttribute(in Name,in BoneName,in Default,out Value,out Found)"}.Value
   )
  ]}
 New Function_ContainedGraph {nodes [
   n1 = RigVMFunctionEntryNode()
   RigVMFunction_MathQuaternionToEuler = &RigVMFunction_MathQuaternionToEuler(
    RotationOrder=XYZ
    Value=&RigUnit_GetRelativeTransformForItem(
     Child=(Type=Bone,Name="hand_l")
     Parent=(Type=Null,Name="lowerarm_l")
     bChildInitial=False
     bParentInitial=False
     Child.Name=@n1.Child
     Parent.Name=@n1.Parent
    ).RelativeTransform.Rotation
   )
   Remap = &RigVMFunction_MathDoubleRemap(
    TargetMaximum=1
    TargetMinimum=0
    Value=@RigVMFunction_MathQuaternionToEuler.Result.Y
    SourceMaximum=@n1.Source Maximum
    SourceMinimum=@n1.Source Minimum
   )
   > RigVMFunctionReturnNode(Delta Rotation=@Remap.Result)
  ]}
 New Function_ContainedGraph {nodes [
   n1 = RigVMFunctionEntryNode()
   RigVMFunction_MathQuaternionToEuler = &RigVMFunction_MathQuaternionToEuler(
    RotationOrder=XYZ
    Value=&RigUnit_GetRelativeTransformForItem(
     Child=(Type=Bone,Name="hand_l")
     Parent=(Type=Null,Name="lowerarm_l")
     bChildInitial=False
     bParentInitial=False
     Child.Name=@n1.Child
     Parent.Name=@n1.Parent
    ).RelativeTransform.Rotation
   )
   Remap = &RigVMFunction_MathDoubleRemap(
    TargetMaximum=1
    TargetMinimum=0
    Value=@RigVMFunction_MathQuaternionToEuler.Result.Z
    SourceMaximum=@n1.Source Maximum
    SourceMinimum=@n1.Source Minimum
   )
   > RigVMFunctionReturnNode(Delta Rotation=@Remap.Result)
  ]}
 New Function_ContainedGraph {nodes [
   RigVMFunctionEntryNode()
   > &RigUnit_ParentConstraint(
    AdvancedSettings=(InterpolationType=Average,RotationOrderForFilter=XZY)
    Child=(Type=Bone,Name="base_shoulderpad_r")
    Parents=((Item=(Type=Bone,Name="upperarm_twist_01_r"),Weight=0.5),(Item=(Type=Bone,Name="clavicle_r"),Weight=0.5))
    Weight=1
   )
   > &RigUnit_ParentConstraint(
    AdvancedSettings=(RotationOrderForFilter=XZY)
    Child=(Type=Bone,Name="base_charm_01_l")
    Parents=((Item=(Type=Bone,Name="spine_01"),Weight=0.0132),(Item=(Type=Bone,Name="spine_03"),Weight=0.0197),(Item=(Type=Bone,Name="spine_04"),Weight=0.0389),(Item=(Type=Bone,Name="spine_02"),Weight=0.2824),(Item=(Type=Bone,Name="pelvis"),Weight=0.6458))
    Weight=1
   )
   > &RigUnit_ParentConstraint(
    AdvancedSettings=(InterpolationType=Average,RotationOrderForFilter=XZY)
    Child=(Type=Bone,Name="base_side_chain_01")
    Parents=((Item=(Type=Bone,Name="pelvis"),Weight=1))
    Weight=1
   )
   > &RigUnit_ParentConstraint(
    AdvancedSettings=(RotationOrderForFilter=XZY)
    Child=(Type=Bone,Name="Base_skirt_Collider")
    Parents="((Item=(Type=Bone,Name=\"thigh_twist_01_r\"),Weight=0.0263),(Item=(Type=Bone,Name=\"deform_groin_r\"),Weight=0.0395),(Item=(Type=Bone,Name=\"dyn_side_skirt_l\"),Weight=0.0425),(Item=(Type=Bone,Name=\"dyn_front_skirt_r\"),Weight=0.0617),(Item=(Type=Bone,Name=\"thigh ...(truncated)"
    Weight=1
   )
   > RigVMFunctionReturnNode()
   &RigUnit_ParentConstraint(
    AdvancedSettings=(RotationOrderForFilter=XZY)
    Child=(Type=Bone,Name="base_side_chain_01")
    ExecutePin=()
    Parents=((Item=(Type=Bone,Name="spine_01"),Weight=0.0371),(Item=(Type=Bone,Name="spine_03"),Weight=0.0721),(Item=(Type=Bone,Name="spine_04"),Weight=0.1668),(Item=(Type=Bone,Name="pelvis"),Weight=0.2732),(Item=(Type=Bone,Name="spine_02"),Weight=0.4507))
    Weight=1
   )
  ]}
 New Function_ContainedGraph {nodes [
   RigVMFunctionEntryNode()
   RigVMFunction_MathBoolToFloat = &RigVMFunction_MathBoolToFloat(Value=RigVMVariableNode() {variable=isEmoting}.Value)
   AlphaInterp = &RigVMFunction_AlphaInterp(
    InterpSpeedDecreasing=10
    InterpSpeedIncreasing=10
    OutRange=(Min=1,Max=0)
    Scale=1
    bClampResult=true
    bInterpResult=False
    bMapRange=true
    Value=@RigVMFunction_MathBoolToFloat.Result
   )
   Multiply = &RigVMFunction_MathDoubleMul(
    A=@AlphaInterp.Result
    B=RigVMDispatchNode(BoneName=root, Default=0, Name=deformRig_l_upperarm_up_alpha) {template="DISPATCH_RigDispatch_GetAnimAttribute(in Name,in BoneName,in Default,out Value,out Found)"}.Value
   )
   > &RigUnit_ModifyTransforms(
    ItemToModify=((Item=(Type=Bone,Name="upperarm_twist_02_l"),Transform=(Rotation=(X=0,Y=0,Z=0,W=-1),Translation=(X=0,Y=-2.6777,Z=0),Scale3D=(X=1,Y=1,Z=1))))
    WeightMinimum=0
    Weight=@Multiply.Result
   )
   > &RigUnit_ParentConstraint(
    AdvancedSettings=(InterpolationType=Average,RotationOrderForFilter=XZY)
    Child=(Type=Bone,Name="lowerarm_twist_02_l")
    Filter=(TranslationFilter=(bX=True,bY=True,bZ=True),RotationFilter=(bX=false,bY=false,bZ=false),ScaleFilter=(bX=True,bY=True,bZ=True))
    Parents=((Item=(Type=Bone,Name="lowerarm_l"),Weight=1))
    Weight=1
    Parents.0.Weight=&RigVMFunction_MathBoolToFloat(Value=RigVMVariableNode() {variable=isEmoting}.Value).Result
   )
   > RigVMFunctionReturnNode()
  ]}
 RigVMFunctionLibrary {nodes [
   Constraints = RigVMCollapseNode(ExecuteContext=())
   Delta Rotation Y = RigVMCollapseNode(ExecuteContext=())
   Delta Rotation Z = RigVMCollapseNode(ExecuteContext=())
   TwistFixes = RigVMCollapseNode(ExecuteContext=())
   Waist objects = RigVMCollapseNode(ExecuteContext=())
  ]}
 RigVMModel {nodes [
   RigUnit_BeginExecution = &RigUnit_BeginExecution()
   FN_DeformRig_FNC = RigVMFunctionReferenceNode(
    DeformRig_ Curve="(deformRig_l_upperarm_up_alpha_72_#1=true,deformRig_r_upperarm_up_alpha_73_#2=true,deformRig_l_upperarm_fr_alpha_74_#3=true,deformRig_r_upperarm_fr_alpha_75_CB51B80C4 ...(truncated)"
    DeformRig_Bool="(upperarm_1_#4=True,elbow_3_#5=True,wrist_7_#6=True,groin_8_#7=True,glute_17_#8=True,knee_9_40C408324CA261 ...(truncated)"
    DeformRig_Debug=(upperarm_debug_18_#4=False,wrist_debug_20_#6=False,lat_debug_26_#9=False,groinGlute_debug_27_#7=False)
    DeformRig_Value="(upperarm_up_18_#10=(Rotation=(X=0.0,Y=-0.0,Z=0.0,W=1.0),Translation=(X=0.0,Y=0.0,Z=8.0),Scale3D=(X=0.70,Y=1.0,Z=1.0)),upperarm_fwd_fr_19_#11=(Rotation=(X=0.0,Y=-0.0,Z=0.0,W=1.0),Translation=(X=0.0, ...(truncated)"
    ExecuteContext=@RigUnit_BeginExecution.ExecutePin
   )
   > Constraints = RigVMFunctionReferenceNode()
   > TwistFixes = RigVMFunctionReferenceNode()
   > Waist objects = RigVMFunctionReferenceNode()
   n1 = RigVMDispatchNode(BoneName=root, Default=0, Name=deformRig_r_upperarm_fr_alpha) {template="DISPATCH_RigDispatch_GetAnimAttribute(in Name,in BoneName,in Default,out Value,out Found)"}
   > &RigUnit_ModifyTransforms(
    ItemToModify="((Item=(Type=Bone,Name=\"dyn_shoulderpad_r\"),Transform=(Rotation=(X=-0.1233,Y=-0.0268,Z=-0.2329,W=0.9643),Translation=(X=0.764,Y=0.0582,Z=-0.0153),Scale3D=(X=1,Y=1.1,Z=1))),(Item=(Type=Bone,Name=\"base_padchains_14_r\"),Transform=(Rotation=(X=0,Y=0,Z=0,W=1),T ...(truncated)"
    WeightMinimum=0
    Weight=@n1.Value
   )
   > &RigUnit_ModifyTransforms(
    ItemToModify="((Item=(Type=Bone,Name=\"dyn_shoulderpad_r\"),Transform=(Rotation=(X=0,Y=0.0147,Z=0,W=-0.9999),Translation=(X=0.0515,Y=0.5121,Z=2.4667),Scale3D=(X=1,Y=1,Z=1))),(Item=(Type=Bone,Name=\"base_padchains_06_r\"),Transform=(Rotation=(X=0,Y=0.0767,Z=0,W=-0.9971),Tran ...(truncated)"
    WeightMinimum=0
    Weight=&RigVMFunction_MathDoubleSub(
     A=RigVMDispatchNode(BoneName=root, Default=0, Name=deformRig_r_upperarm_up_alpha) {template="DISPATCH_RigDispatch_GetAnimAttribute(in Name,in BoneName,in Default,out Value,out Found)"}.Value
     B=@n1.Value
    ).Result
   )
   > n2 = &RigUnit_SphericalPoseReader(
    ActiveRegionScaleFactors=(PositiveWidth=1,NegativeWidth=1,PositiveHeight=1,NegativeHeight=0.8)
    ActiveRegionSize=0.1
    Debug=(bDrawDebug=false,bDraw2D=False,bDrawLocalAxes=False,DebugScale=25,DebugSegments=20,DebugThickness=0.25)
    DriverAxis=(X=1,Y=0,Z=0)
    DriverItem=(Type=Bone,Name="upperarm_r")
    FalloffRegionScaleFactors=(PositiveWidth=1,NegativeWidth=1,PositiveHeight=1.7,NegativeHeight=0.5)
    FalloffSize=0.2
    FlipHeightScaling=False
    FlipWidthScaling=False
    OptionalParentItem=(Type=Bone,Name="None")
    RotationOffset=(X=51.9395,Y=8.7023,Z=22.4256)
   )
   RigVMFunction_MathBoolToFloat = &RigVMFunction_MathBoolToFloat(Value=RigVMVariableNode() {variable=isSkydive}.Value)
   AlphaInterp = &RigVMFunction_AlphaInterp(
    InterpSpeedDecreasing=10
    InterpSpeedIncreasing=10
    OutRange=(Min=1,Max=0)
    Scale=1
    bClampResult=true
    bInterpResult=False
    bMapRange=true
    Value=@RigVMFunction_MathBoolToFloat.Result
   )
   Multiply = &RigVMFunction_MathDoubleMul(B=@AlphaInterp.Result, A=@n2.OutputParam)
   > &RigUnit_ModifyTransforms(
    ItemToModify="((Item=(Type=Bone,Name=\"dyn_shoulderpad_r\"),Transform=(Rotation=(X=0,Y=0,Z=0.1207,W=-0.9927),Translation=(X=-0.0079,Y=-1.1803,Z=1.3796),Scale3D=(X=1,Y=1,Z=1))),(Item=(Type=Bone,Name=\"base_padchains_06_r\"),Transform=(Rotation=(X=0,Y=0,Z=-0.0965,W=-0.9953),T ...(truncated)"
    WeightMinimum=0
    Weight=@Multiply.Result
   )
   > &RigUnit_ModifyTransforms(
    ItemToModify=((Item=(Type=Bone,Name="deform_upperarm_bk_r"),Transform=(Rotation=(X=0,Y=0,Z=0,W=1),Translation=(X=0,Y=-3,Z=0),Scale3D=(X=1,Y=1,Z=1))))
    WeightMinimum=0
    Weight=RigVMDispatchNode(BoneName=root, Default=0, Name=deformRig_r_upperarm_fr_alpha) {template="DISPATCH_RigDispatch_GetAnimAttribute(in Name,in BoneName,in Default,out Value,out Found)"}.Value
   )
   > &RigUnit_ModifyTransforms(
    ItemToModify=((Item=(Type=Bone,Name="base_shoulderpad_r"),Transform=(Rotation=(X=0,Y=0,Z=0.0555,W=0.9985),Translation=(X=-0.1421,Y=1.2747,Z=0.8091),Scale3D=(X=1,Y=1,Z=1))))
    WeightMinimum=0
    Weight=RigVMDispatchNode(BoneName=root, Default=0, Name=deformRig_r_upperarm_bk_alpha) {template="DISPATCH_RigDispatch_GetAnimAttribute(in Name,in BoneName,in Default,out Value,out Found)"}.Value
   )
   &RigUnit_ModifyTransforms(
    ExecutePin=()
    ItemToModify=((Item=(Type=Bone,Name="upperarm_twist_02_r"),Transform=(Rotation=(X=0.6225,Y=0,Z=0,W=0.7826),Translation=(X=0,Y=0,Z=0),Scale3D=(X=1,Y=1,Z=1))))
    WeightMinimum=0
    Weight=&RigVMFunction_MathBoolToFloat(Value=RigVMVariableNode() {variable=Parachuteopen}.Value).Result
   )
   &RigUnit_PrepareForExecution()
   addFKCtrl = RigVMFunctionReferenceNode(
    Custom_FKCtrl=()
    ExecuteContext=()
    Options_FKCtrl="(pelvis_1_#12=false,spine_3_#13=false,clavicle_5_#14=true,arm_7_#15=true,leg_9_#16=true,head_11_46DC15EF465CD29 ...(truncated)"
    Shape_Settings="(pelvis_2_#17=(bVisible=True,Name=\"Hexagon_Thin\",Color=(R=0.0117,G=0.8281,B=1,A=1),Transform=(Rotation=(X=0,Y=-0.7071,Z=0,W=0.7071),Translation=(X=2,Y=0,Z=0),Scale3D=(X=5,Y=5,Z=7))),spine_4_#18=(bVi ...(truncated)"
   )
   addNullObject = RigVMFunctionReferenceNode(
    Custom_Joints=()
    ExecuteContext=()
    Options_Null="(pelvis_1_#12=False,spine_3_#13=False,clavicle_5_#14=False,arm_7_#15=true,leg_9_#16=False,head_11_46DC15EF465CD ...(truncated)"
   )
   setFKCtrl = RigVMFunctionReferenceNode(ExecuteContext=(), setFK_Ctrl=false)
  ]}
]
hierarchy [
 root {loc=<0 0 0>, rot=<0 0 0 1>}
 pelvis {parent=root, loc=<0 -0.5301 98.0571>, rot=<0.0127 0.707 -0.0127 -0.707>}
 spine_01 {parent=pelvis, loc=<6.1208 -0.9328 0>, rot=<0 0 0.034 -0.9994>}
 spine_02 {parent=spine_01, loc=<6.5997 -0.2095 0>, rot=<0 0 0.0238 -0.9997>}
 spine_03 {parent=spine_02, loc=<6.4914 -0.1237 0>, rot=<0 0 0.0149 -0.9999>}
 spine_04 {parent=spine_03, loc=<6.9598 0.0794 0>, rot=<0 0 -0.0132 -0.9999>}
 spine_05 {parent=spine_04, loc=<6.9987 0 0>, rot=<0 0 -0.0015 -1>}
 neck_01 {parent=spine_05, loc=<11.3138 0.1529 0>, rot=<0 0 -0.2003 -0.9797>}
 neck_02 {parent=neck_01, loc=<4.5627 0 0>, rot=<0 0 0.0504 -0.9987>}
 head {parent=neck_02, loc=<5.3914 0 0>, rot=<0 0 0.115 -0.9934>}
 clavicle_l {parent=spine_05, loc=<5.4831 -0.1053 -3.2946>, rot=<0.0465 -0.7289 -0.009 -0.6829>}
 upperarm_l {parent=clavicle_l, loc=<9.8658 -1.2136 -0.4105>, rot=<0.0164 -0.3245 -0.0269 -0.9454>}
 lowerarm_l {parent=upperarm_l, loc=<26.465 -0.8345 0.259>, rot=<0.0036 0.0161 -0.2298 -0.9731>}
 hand_l {parent=lowerarm_l, loc=<22.7355 -0.3203 -0.6105>, rot=<-0.6784 -0.0223 0.0735 -0.7307>}
 thumb_01_l {parent=hand_l, loc=<1.3118 -0.3191 -2.459>, rot=<0.6546 -0.4042 0.0664 -0.6354>}
 thumb_02_l {parent=thumb_01_l, loc=<5.9336 -0.0448 0.0401>, rot=<-0.0037 0.0213 0.298 -0.9543>}
 thumb_03_l {parent=thumb_02_l, loc=<2.2223 0.3682 -0.0907>, rot=<0.0118 -0.013 0.0066 -0.9998>}
 index_metacarpal_l {parent=hand_l, loc=<1.6535 0.0183 -2.96>, rot=<0.0158 -0.0124 -0.0173 -0.9996>}
 index_01_l {parent=index_metacarpal_l, loc=<8.3008 0.0357 -0.0524>, rot=<-0.0044 0.0485 0.0924 -0.9945>}
 index_02_l {parent=index_01_l, loc=<3.3587 -0.3004 0.0452>, rot=<-0.002 0.0027 0.3225 -0.9466>}
 index_03_l {parent=index_02_l, loc=<2.5578 -0.123 -0.0252>, rot=<-0.002 0.0067 0.2558 -0.9667>}
 middle_metacarpal_l {parent=hand_l, loc=<2.1856 -0.1725 -0.5984>, rot=<0.006 0.0168 -0.0329 -0.9993>}
 middle_01_l {parent=middle_metacarpal_l, loc=<7.3275 -0.2447 0.0129>, rot=<-0.0208 0.0042 0.0858 -0.9961>}
 middle_02_l {parent=middle_01_l, loc=<4.0123 0.1401 0.1175>, rot=<-0.0218 0.0258 0.3512 -0.9357>}
 middle_03_l {parent=middle_02_l, loc=<2.6077 0.0534 0.1373>, rot=<0.0082 0.0188 0.1708 -0.9851>}
 ring_metacarpal_l {parent=hand_l, loc=<2.2074 -0.3625 1.3638>, rot=<-0.0284 0.0722 -0.0458 -0.9959>}
 ring_01_l {parent=ring_metacarpal_l, loc=<7.0295 -0.6197 -0.1446>, rot=<-0.0489 -0.0252 0.1892 -0.9804>}
 ring_02_l {parent=ring_01_l, loc=<3.6174 0.4647 -0.1027>, rot=<0 0 0.2806 -0.9598>}
 ring_03_l {parent=ring_02_l, loc=<2.4152 0.114 0.0199>, rot=<0 0 0.1189 -0.9929>}
 pinky_metacarpal_l {parent=hand_l, loc=<1.7616 -0.9753 3.3814>, rot=<-0.0397 0.0814 -0.069 -0.9935>}
 pinky_01_l {parent=pinky_metacarpal_l, loc=<7.1709 -0.4831 -0.1094>, rot=<-0.0654 -0.0035 0.2242 -0.9723>}
 pinky_02_l {parent=pinky_01_l, loc=<2.8274 0.0893 -0.1043>, rot=<0.0037 0.0011 0.2854 -0.9584>}
 pinky_03_l {parent=pinky_02_l, loc=<2.2046 0.0307 -0.0078>, rot=<-0.0002 0.0098 0.1445 -0.9895>}
 deform_wrist_dn_base_l {parent=hand_l, loc=<0 -1.5326 0>, rot=<0 0 0 -1>}
 deform_wrist_dn_l {parent=deform_wrist_dn_base_l, loc=<0 0 0>, rot=<0 0 0 1>}
 deform_wrist_up_base_l {parent=hand_l, loc=<0 2.3009 0>, rot=<0 0 0 -1>}
 deform_wrist_up_l {parent=deform_wrist_up_base_l, loc=<0 0 0>, rot=<0 0 0 1>}
 lowerarm_twist_02_l {parent=lowerarm_l, loc=<10.6707 0 0>, rot=<0 0 0 1>}
 lowerarm_twist_01_l {parent=lowerarm_l, loc=<18.316 0 0>, rot=<0 0 0 1>}
 deform_elbow_in_base_l {parent=lowerarm_l, loc=<0.1834 3.6062 -0.2572>, rot=<0 0 0 -1>}
 deform_elbow_in_l {parent=deform_elbow_in_base_l, loc=<0 0 0>, rot=<0 0 0 1>}
 deform_elbow_out_base_l {parent=lowerarm_l, loc=<0 -4 0>, rot=<0 0 0 1>}
 deform_elbow_out_l {parent=deform_elbow_out_base_l, loc=<0 0 0>, rot=<0 0 0 1>}
 upperarm_twist_01_l {parent=upperarm_l, loc=<7.2135 -0.1134 0>, rot=<0 0 0 1>}
 deform_upperarm_bk_base_l {parent=upperarm_twist_01_l, loc=<-6.4053 -4.3746 0.314>, rot=<0 0 0 -1>}
 deform_upperarm_bk_l {parent=deform_upperarm_bk_base_l, loc=<0 0 0>, rot=<0 0 0 1>}
 deform_upperarm_fr_base_l {parent=upperarm_twist_01_l, loc=<-5.6312 5.092 0.3354>, rot=<0 0 0 -1>}
 deform_upperarm_fr_l {parent=deform_upperarm_fr_base_l, loc=<0 0 0>, rot=<0 0 0 1>}
 deform_upperarm_up_base_l {parent=upperarm_twist_01_l, loc=<-8.891 0.4825 3.9517>, rot=<0 0 0 -1>}
 deform_upperarm_up_l {parent=deform_upperarm_up_base_l, loc=<0 0 0>, rot=<0 0 0 1>}
 upperarm_twist_02_l {parent=upperarm_l, loc=<19.7998 -0.1134 0>, rot=<0 0 0 1>}
 clavicle_r {parent=spine_05, loc=<5.4831 -0.1053 3.2946>, rot=<0.7289 0.0465 -0.6829 0.009>}
 upperarm_r {parent=clavicle_r, loc=<-9.8658 1.2136 0.4105>, rot=<0.0164 -0.3245 -0.0269 -0.9454>}
 lowerarm_r {parent=upperarm_r, loc=<-26.465 0.8345 -0.259>, rot=<0.0036 0.0161 -0.2298 -0.9731>}
 hand_r {parent=lowerarm_r, loc=<-22.7355 0.3203 0.6106>, rot=<-0.6784 -0.0223 0.0735 -0.7307>}
 thumb_01_r {parent=hand_r, loc=<-1.3118 0.3191 2.459>, rot=<0.6546 -0.4042 0.0664 -0.6354>}
 thumb_02_r {parent=thumb_01_r, loc=<-5.9336 0.0448 -0.0401>, rot=<-0.0037 0.0213 0.298 -0.9543>}
 thumb_03_r {parent=thumb_02_r, loc=<-2.2223 -0.3682 0.0907>, rot=<0.0118 -0.013 0.0066 -0.9998>}
 index_metacarpal_r {parent=hand_r, loc=<-1.6535 -0.0183 2.96>, rot=<0.0158 -0.0124 -0.0173 -0.9996>}
 index_01_r {parent=index_metacarpal_r, loc=<-8.3008 -0.0357 0.0524>, rot=<-0.0044 0.0485 0.0924 -0.9945>}
 index_02_r {parent=index_01_r, loc=<-3.3587 0.3004 -0.0452>, rot=<-0.002 0.0027 0.3225 -0.9466>}
 index_03_r {parent=index_02_r, loc=<-2.5578 0.123 0.0252>, rot=<-0.002 0.0067 0.2558 -0.9667>}
 middle_metacarpal_r {parent=hand_r, loc=<-2.1856 0.1725 0.5984>, rot=<0.006 0.0168 -0.0329 -0.9993>}
 middle_01_r {parent=middle_metacarpal_r, loc=<-7.3275 0.2447 -0.0129>, rot=<-0.0208 0.0042 0.0858 -0.9961>}
 middle_02_r {parent=middle_01_r, loc=<-4.0123 -0.1401 -0.1175>, rot=<-0.0218 0.0258 0.3512 -0.9357>}
 middle_03_r {parent=middle_02_r, loc=<-2.6077 -0.0534 -0.1373>, rot=<0.0082 0.0188 0.1708 -0.9851>}
 ring_metacarpal_r {parent=hand_r, loc=<-2.2074 0.3625 -1.3638>, rot=<-0.0284 0.0722 -0.0458 -0.9959>}
 ring_01_r {parent=ring_metacarpal_r, loc=<-7.0295 0.6197 0.1446>, rot=<-0.0489 -0.0252 0.1892 -0.9804>}
 ring_02_r {parent=ring_01_r, loc=<-3.6174 -0.4647 0.1027>, rot=<0 0 0.2806 -0.9598>}
 ring_03_r {parent=ring_02_r, loc=<-2.4152 -0.114 -0.0199>, rot=<0 0 0.1189 -0.9929>}
 pinky_metacarpal_r {parent=hand_r, loc=<-1.7616 0.9753 -3.3814>, rot=<-0.0397 0.0814 -0.069 -0.9935>}
 pinky_01_r {parent=pinky_metacarpal_r, loc=<-7.1709 0.4831 0.1094>, rot=<-0.0654 -0.0035 0.2242 -0.9723>}
 pinky_02_r {parent=pinky_01_r, loc=<-2.8274 -0.0893 0.1043>, rot=<0.0037 0.0011 0.2854 -0.9584>}
 pinky_03_r {parent=pinky_02_r, loc=<-2.2046 -0.0307 0.0078>, rot=<-0.0002 0.0098 0.1445 -0.9895>}
 deform_wrist_dn_base_r {parent=hand_r, loc=<-0.0001 1.5328 0>, rot=<0 0 0 -1>}
 deform_wrist_dn_r {parent=deform_wrist_dn_base_r, loc=<0 0 0>, rot=<0 0 0 1>}
 deform_wrist_up_base_r {parent=hand_r, loc=<-0.0002 -2.3007 -0.0001>, rot=<0 0 0 -1>}
 deform_wrist_up_r {parent=deform_wrist_up_base_r, loc=<0 0 0>, rot=<0 0 0 1>}
 lowerarm_twist_02_r {parent=lowerarm_r, loc=<-10.6707 0 0>, rot=<0 0 0 -1>}
 lowerarm_twist_01_r {parent=lowerarm_r, loc=<-18.316 0 0>, rot=<0 0 0 -1>}
 deform_elbow_in_base_r {parent=lowerarm_r, loc=<-0.1834 -3.6062 0.2572>, rot=<0 0 0 -1>}
 deform_elbow_in_r {parent=deform_elbow_in_base_r, loc=<0 0 0>, rot=<0 0 0 1>}
 deform_elbow_out_base_r {parent=lowerarm_r, loc=<0 4 -0.0001>, rot=<0 0 0 -1>}
 deform_elbow_out_r {parent=deform_elbow_out_base_r, loc=<0 0 0>, rot=<0 0 0 1>}
 upperarm_twist_01_r {parent=upperarm_r, loc=<-7.2135 0.1134 0>, rot=<0 0 0 -1>}
 deform_upperarm_bk_base_r {parent=upperarm_twist_01_r, loc=<6.4053 4.3746 -0.314>, rot=<0 0 0 -1>}
 deform_upperarm_bk_r {parent=deform_upperarm_bk_base_r, loc=<0 0 0>, rot=<0 0 0 1>}
 deform_upperarm_fr_base_r {parent=upperarm_twist_01_r, loc=<5.6312 -5.092 -0.3354>, rot=<0 0 0 -1>}
 deform_upperarm_fr_r {parent=deform_upperarm_fr_base_r, loc=<0 0 0>, rot=<0 0 0 1>}
 deform_upperarm_up_base_r {parent=upperarm_twist_01_r, loc=<8.891 -0.4825 -3.9517>, rot=<0 0 0 -1>}
 deform_upperarm_up_r {parent=deform_upperarm_up_base_r, loc=<0 0 0>, rot=<0 0 0 1>}
 base_shoulderpad_r {parent=upperarm_twist_01_r, loc=<7.393 -0.4763 -2.2834>, rot=<0 -1 0 0>}
 dyn_shoulderpad_r {parent=base_shoulderpad_r, loc=<0 0 0>, rot=<0 0 0 1>}
 base_padchains_01_r {parent=dyn_shoulderpad_r, loc=<-3.7943 -3.6408 -1.9252>, rot=<-0.4217 -0.0984 0.5068 -0.7454>}
 dyn_padchains_01_r {parent=base_padchains_01_r, loc=<0 0 0>, rot=<0 0 0 1>}
 base_padchains_02_r {parent=dyn_shoulderpad_r, loc=<-3.2141 -3.9947 -2.0831>, rot=<-0.4585 -0.1196 0.4385 -0.7637>}
 dyn_padchains_02_r {parent=base_padchains_02_r, loc=<0 0 0>, rot=<0 0 0 1>}
 base_padchains_03_r {parent=dyn_shoulderpad_r, loc=<-1.9934 -4.3261 -2.1933>, rot=<-0.4662 -0.1925 0.3482 -0.7902>}
 dyn_padchains_03_r {parent=base_padchains_03_r, loc=<0 0 0>, rot=<0 0 0 1>}
 base_padchains_04_r {parent=dyn_shoulderpad_r, loc=<-0.453 -4.539 -2.3314>, rot=<-0.5154 -0.2332 0.3104 -0.7639>}
 dyn_padchains_04_r {parent=base_padchains_04_r, loc=<0 0 0>, rot=<0 0 0 1>}
 base_padchains_05_r {parent=dyn_shoulderpad_r, loc=<1.3254 -4.3756 -2.1722>, rot=<-0.4822 -0.3016 0.275 -0.7752>}
 dyn_padchains_05_r {parent=base_padchains_05_r, loc=<0 0 0>, rot=<0 0 0 1>}
 base_padchains_06_r {parent=dyn_shoulderpad_r, loc=<2.4315 -3.6708 -1.6496>, rot=<-0.481 -0.2487 0.2856 -0.7907>}
 dyn_padchains_06_r {parent=base_padchains_06_r, loc=<0 0 0>, rot=<0 0 0 1>}
 base_padchains_07_r {parent=dyn_shoulderpad_r, loc=<3.0069 -3.1938 -1.3481>, rot=<-0.5488 -0.2271 0.3037 -0.745>}
 dyn_padchains_07_r {parent=base_padchains_07_r, loc=<0 0 0>, rot=<0 0 0 1>}
 base_padchains_08_r {parent=dyn_shoulderpad_r, loc=<4.4511 -2.2047 -0.1233>, rot=<-0.3871 -0.2526 0.2322 -0.8558>}
 dyn_padchains_08_r {parent=base_padchains_08_r, loc=<0 0 0>, rot=<0 0 0 1>}
 dyn_padchains_08_01_r {parent=base_padchains_08_r, loc=<0 0 0>, rot=<0 0 0 1>}
 dyn_padchains_08_02_r {parent=dyn_padchains_08_01_r, loc=<5.898 -0.3582 0.1609>, rot=<0 0 0 1>}
 base_padchains_09_r {parent=dyn_shoulderpad_r, loc=<5.1963 -1.526 0.4996>, rot=<-0.2886 -0.1777 0.1571 -0.9276>}
 dyn_padchains_09_r {parent=base_padchains_09_r, loc=<0 0 0>, rot=<0 0 0 1>}
 dyn_padchains_09_01_r {parent=base_padchains_09_r, loc=<0 0 0>, rot=<0 0 0 1>}
 dyn_padchains_09_02_r {parent=dyn_padchains_09_01_r, loc=<6.0405 -0.1135 0.2466>, rot=<0 0 0 1>}
 base_padchains_10_r {parent=dyn_shoulderpad_r, loc=<5.7663 -0.6815 0.9917>, rot=<-0.1082 -0.0978 0.0729 -0.9866>}
 dyn_padchains_10_r {parent=base_padchains_10_r, loc=<0 0 0>, rot=<0 0 0 1>}
 dyn_padchains_10_01_r {parent=base_padchains_10_r, loc=<0 0 0>, rot=<0 0 0 1>}
 dyn_padchains_10_02_r {parent=dyn_padchains_10_01_r, loc=<6.61 -0.0889 0.0782>, rot=<0 0 0 1>}
 base_padchains_11_r {parent=dyn_shoulderpad_r, loc=<6.1766 0.5802 1.3015>, rot=<0.0181 -0.037 -0.0126 -0.9991>}
 dyn_padchains_11_r {parent=base_padchains_11_r, loc=<0 0 0>, rot=<0 0 0 1>}
 dyn_padchains_11_01_r {parent=base_padchains_11_r, loc=<0 0 0>, rot=<0 0 0 1>}
 dyn_padchains_11_02_r {parent=dyn_padchains_11_01_r, loc=<6.7692 -0.0565 0.1679>, rot=<0 0 0 1>}
 base_padchains_12_r {parent=dyn_shoulderpad_r, loc=<5.8752 1.6861 1.0833>, rot=<0.201 -0.0446 -0.1267 -0.9703>}
 dyn_padchains_12_r {parent=base_padchains_12_r, loc=<0 0 0>, rot=<0 0 0 1>}
 dyn_padchains_12_01_r {parent=base_padchains_12_r, loc=<0 0 0>, rot=<0 0 0 1>}
 dyn_padchains_12_02_r {parent=dyn_padchains_12_01_r, loc=<5.881 0 0.1152>, rot=<0 0 0 1>}
 base_padchains_13_r {parent=dyn_shoulderpad_r, loc=<5.4238 2.3536 0.7673>, rot=<0.3568 -0.1135 -0.227 -0.899>}
 dyn_padchains_13_r {parent=base_padchains_13_r, loc=<0 0 0>, rot=<0 0 0 1>}
 dyn_padchains_13_01_r {parent=base_padchains_13_r, loc=<0 0 0>, rot=<0 0 0 1>}
 dyn_padchains_13_02_r {parent=dyn_padchains_13_01_r, loc=<5.8465 -0.1097 0.1828>, rot=<0 0 0 1>}
 base_padchains_14_r {parent=dyn_shoulderpad_r, loc=<4.5768 3.0775 0.0484>, rot=<0.4544 -0.1569 -0.2353 -0.8447>}
 dyn_padchains_14_r {parent=base_padchains_14_r, loc=<0 0 0>, rot=<0 0 0 1>}
 dyn_padchains_14_02_r {parent=dyn_padchains_14_r, loc=<5.766 0.1752 0.3617>, rot=<0.0384 0.0014 -0.0367 -0.9986>}
 base_padchains_15_r {parent=dyn_shoulderpad_r, loc=<3.8726 3.4668 -0.5774>, rot=<0.4967 -0.2421 -0.315 -0.7717>}
 dyn_padchains_15_r {parent=base_padchains_15_r, loc=<0 0 0>, rot=<0 0 0 1>}
 base_padchains_16_r {parent=dyn_shoulderpad_r, loc=<3.4278 3.7084 -0.9712>, rot=<0.5291 -0.255 -0.3153 -0.7454>}
 dyn_padchains_16_r {parent=base_padchains_16_r, loc=<0 0 0>, rot=<0 0 0 1>}
 base_padchains_17_r {parent=dyn_shoulderpad_r, loc=<1.6474 4.3033 -2.373>, rot=<0.5745 -0.2545 -0.3151 -0.7113>}
 dyn_padchains_17_r {parent=base_padchains_17_r, loc=<0 0 0>, rot=<0 0 0 1>}
 base_padchains_18_r {parent=dyn_shoulderpad_r, loc=<0.993 4.4232 -2.4758>, rot=<0.6203 -0.2602 -0.3034 -0.675>}
 dyn_padchains_18_r {parent=base_padchains_18_r, loc=<0 0 0>, rot=<0 0 0 1>}
 base_padchains_19_r {parent=dyn_shoulderpad_r, loc=<-0.2053 4.3688 -2.5057>, rot=<0.654 -0.2511 -0.3316 -0.6319>}
 dyn_padchains_19_r {parent=base_padchains_19_r, loc=<0 0 0>, rot=<0 0 0 1>}
 base_padchains_20_r {parent=dyn_shoulderpad_r, loc=<-0.8837 4.1041 -2.3258>, rot=<0.6049 -0.2793 -0.2935 -0.6856>}
 dyn_padchains_20_r {parent=base_padchains_20_r, loc=<0 0 0>, rot=<0 0 0 1>}
 upperarm_twist_02_r {parent=upperarm_r, loc=<-19.7998 0.1134 0>, rot=<0 0 0 -1>}
 attach_backpack {parent=spine_05, loc=<-0.123 -13.473 0>, rot=<-0.0093 -0.707 0.0093 -0.707>}
 Backpack_BR {parent=attach_backpack, type=Null, loc=<0 -9.4635 0>, rot=<-0.0379 0 0 0.9993>}
 attach_cape {parent=spine_05, loc=<9.017 -11.714 0>, rot=<0.0616 -0.7044 -0.0616 -0.7044>}
 cape {parent=attach_cape, type=Null, loc=<0 -9.0272 -1.578>, rot=<0 0 0 1>}
 deform_pec_aim_base_l {parent=spine_04, loc=<3.2456 9.3309 -1.2237>, rot=<0.0603 -0.4033 0.3318 -0.8507>}
 deform_pec_aim_l {parent=deform_pec_aim_base_l, loc=<0 0 0>, rot=<0 0 0 1>}
 deform_pec_base_l {parent=deform_pec_aim_l, loc=<9.3665 0 0>, rot=<0 0 0 1>}
 deform_pec_l {parent=deform_pec_base_l, loc=<0 0 0>, rot=<0 0 0 1>}
 deform_pec_aim_base_r {parent=spine_04, loc=<3.2452 9.3309 1.2237>, rot=<-0.4033 -0.0603 0.8507 0.3318>}
 deform_pec_aim_r {parent=deform_pec_aim_base_r, loc=<0 0 0>, rot=<0 0 0 1>}
 deform_pec_base_r {parent=deform_pec_aim_r, loc=<-9.3667 -0.0003 -0.0004>, rot=<0 0 0 1>}
 deform_pec_r {parent=deform_pec_base_r, loc=<0 0 0>, rot=<0 0 0 1>}
 base_sideStrap_01_fr {parent=spine_02, loc=<-1.6361 7.2286 -0.29>, rot=<-0.1551 -0.8018 -0.5017 0.2853>}
 dyn_sideStrap_01_fr {parent=base_sideStrap_01_fr, loc=<0 0 0>, rot=<0 0 0 1>}
 dyn_sideStrap_02_fr {parent=dyn_sideStrap_01_fr, loc=<4.79 0 0>, rot=<0.1298 -0.0084 -0.024 -0.9912>}
 dyn_sideStrap_03_fr {parent=dyn_sideStrap_02_fr, loc=<5.7767 0 0>, rot=<-0.0172 -0.1287 -0.1514 -0.9799>}
 dyn_sideStrap_04_fr {parent=dyn_sideStrap_03_fr, loc=<5.7246 0 0>, rot=<0 -0.1736 0 -0.9848>}
 base_sideStrap_01_bck {parent=spine_02, loc=<-1.7805 -4.1448 7.624>, rot=<-0.2325 -0.8459 -0.4111 0.248>}
 dyn_sideStrap_01_bck {parent=base_sideStrap_01_bck, loc=<0 0 0>, rot=<0 0 0 1>}
 dyn_sideStrap_02_bck {parent=dyn_sideStrap_01_bck, loc=<5.3042 0 0>, rot=<0.2321 0.1096 0.0288 -0.9661>}
 dyn_sideStrap_03_bck {parent=dyn_sideStrap_02_bck, loc=<4.8196 0 0>, rot=<0.0116 0.2556 0.039 -0.9659>}
 dyn_sideStrap_04_bck {parent=dyn_sideStrap_03_bck, loc=<3.8539 0 0>, rot=<0 0.1736 0 -0.9848>}
 base_charm_01_l {parent=spine_02, loc=<-3.2633 4.8246 -7.3208>, rot=<-0.0956 -0.7599 -0.6292 -0.1327>}
 dyn_charm_01_l {parent=base_charm_01_l, loc=<0 0 0>, rot=<0 0 0 1>}
 dyn_charm_02_l {parent=dyn_charm_01_l, loc=<5.1873 -0.0153 -0.1041>, rot=<0.006 -0.0858 0.0043 -0.9963>}
 dyn_charm_03_l {parent=dyn_charm_02_l, loc=<3.6157 0 0.0498>, rot=<0 0 0 1>}
 base_side_chain_01 {parent=spine_02, loc=<-2.2592 1.5042 3.6101>, rot=<0.3602 0.2727 0.892 0.0141>}
 dyn_side_chain_01 {parent=base_side_chain_01, loc=<0 0 0>, rot=<0 0 0 1>}
 dyn_side_chain_02 {parent=dyn_side_chain_01, loc=<6.8838 -0.0012 -0.0106>, rot=<0 0 0 1>}
 thigh_l {parent=pelvis, loc=<-1.0855 -0.2109 -8.7053>, rot=<0.0009 0.0355 0.0155 -0.9993>}
 calf_l {parent=thigh_l, loc=<-42.5284 -0.0019 -0.7297>, rot=<0.0006 -0.0161 -0.0391 -0.9991>}
 foot_l {parent=calf_l, loc=<-46.7401 -2.5014 -0.592>, rot=<0.0427 -0.0238 0.0353 -0.9982>}
 ball_l {parent=foot_l, loc=<-3.5658 10.8066 -0.1544>, rot=<0.0001 -0.0001 -0.7063 -0.7079>}
 calf_twist_02_l {parent=calf_l, loc=<-19.1546 -1.7943 -0.2638>, rot=<0.0237 0.0095 -0.0035 -0.9997>}
 calf_twist_01_l {parent=calf_l, loc=<-35.8483 -1.9197 -0.5794>, rot=<0.0237 0.0095 -0.0035 -0.9997>}
 deform_knee_in_base_l {parent=calf_l, loc=<0.8031 -5.0231 0.2456>, rot=<0 0 0 -1>}
 deform_knee_in_l {parent=deform_knee_in_base_l, loc=<0 0 0>, rot=<0 0 0 1>}
 deform_knee_out_base_l {parent=calf_l, loc=<0 4 0>, rot=<0 0 0 -1>}
 deform_knee_out_l {parent=deform_knee_out_base_l, loc=<0 0 0>, rot=<0 0 0 1>}
 thigh_twist_01_l {parent=thigh_l, loc=<-27.952 0 0>, rot=<0.0013 0 0 -1>}
 deform_groin_base_l {parent=thigh_l, loc=<0.4801 5.6776 -0.2486>, rot=<0 0 0 -1>}
 deform_groin_l {parent=deform_groin_base_l, loc=<0 0 0>, rot=<0 0 0 1>}
 base_front_skirt_l {parent=thigh_l, loc=<-2.769 5.9603 0.108>, rot=<-0.0174 -0.7506 -0.6598 0.0315>}
 dyn_front_skirt_l {parent=base_front_skirt_l, loc=<0 0 0>, rot=<0 0 0 1>}
 base_bk_skirt_l {parent=thigh_l, loc=<-2.2654 -9.3977 1.3381>, rot=<0.0178 -0.7672 0.6405 0.0304>}
 dyn_bk_skirt_l {parent=base_bk_skirt_l, loc=<0 0 0>, rot=<0 0 0 1>}
 base_side_skirt_l {parent=thigh_l, loc=<-2.8001 -0.426 -5.6269>, rot=<-0.0048 -0.9991 0.0098 -0.042>}
 dyn_side_skirt_l {parent=base_side_skirt_l, loc=<0 0 0>, rot=<0 0 0 1>}
 thigh_r {parent=pelvis, loc=<-1.0855 -0.2109 8.7053>, rot=<0.0355 -0.0009 0.9993 0.0155>}
 calf_r {parent=thigh_r, loc=<42.5284 0.0019 0.7297>, rot=<0.0006 -0.0161 -0.0391 -0.9991>}
 foot_r {parent=calf_r, loc=<46.7401 2.5014 0.592>, rot=<0.0427 -0.0238 0.0353 -0.9982>}
 ball_r {parent=foot_r, loc=<3.5658 -10.8066 0.1544>, rot=<0.0001 -0.0001 -0.7063 -0.7079>}
 calf_twist_02_r {parent=calf_r, loc=<19.1546 1.7943 0.2638>, rot=<0.0237 0.0095 -0.0035 -0.9997>}
 calf_twist_01_r {parent=calf_r, loc=<35.8483 1.9197 0.5794>, rot=<0.0237 0.0095 -0.0035 -0.9997>}
 deform_knee_in_base_r {parent=calf_r, loc=<-0.8031 5.0231 -0.2457>, rot=<0 0 0 -1>}
 deform_knee_in_r {parent=deform_knee_in_base_r, loc=<0 0 0>, rot=<0 0 0 1>}
 deform_knee_out_base_r {parent=calf_r, loc=<0 -4 0>, rot=<0 0 0 -1>}
 deform_knee_out_r {parent=deform_knee_out_base_r, loc=<0 0 0>, rot=<0 0 0 1>}
 thigh_twist_01_r {parent=thigh_r, loc=<27.952 0 0>, rot=<0.0013 0 0 -1>}
 deform_groin_base_r {parent=thigh_r, loc=<-0.4801 -5.6776 0.2486>, rot=<0 0 0 -1>}
 deform_groin_r {parent=deform_groin_base_r, loc=<0 0 0>, rot=<0 0 0 1>}
 base_bk_skirt_r {parent=thigh_r, loc=<2.272 9.7191 -1.3383>, rot=<0.6405 -0.0304 -0.0178 -0.7672>}
 dyn_bk_skirt_r {parent=base_bk_skirt_r, loc=<0 0 0>, rot=<0 0 0 1>}
 base_front_skirt_r {parent=thigh_r, loc=<2.7694 -5.9596 -0.1077>, rot=<-0.6598 -0.0315 0.0174 -0.7506>}
 dyn_front_skirt_r {parent=base_front_skirt_r, loc=<0 0 0>, rot=<0 0 0 1>}
 base_side_skirt_r {parent=thigh_r, loc=<2.8 0.426 5.6269>, rot=<0.0098 0.042 0.0048 -0.9991>}
 dyn_side_skirt_r {parent=base_side_skirt_r, loc=<0 0 0>, rot=<0 0 0 1>}
 deform_glute_base_l {parent=pelvis, loc=<-7.3466 -9.0213 -9.1245>, rot=<0.0172 0.031 -0.4477 -0.8935>}
 deform_glute_l {parent=deform_glute_base_l, loc=<0 0 0>, rot=<0 0 0 1>}
 deform_glute_base_r {parent=pelvis, loc=<-7.3464 -9.0213 9.1245>, rot=<0.031 -0.0172 0.8935 -0.4477>}
 deform_glute_r {parent=deform_glute_base_r, loc=<0 0 0>, rot=<0 0 0 1>}
 thigh_l_null {parent=pelvis, type=Null, loc=<-1.0855 -0.2109 -8.7053>, rot=<0.0009 0.0355 0.0155 -0.9993>}
 thigh_r_null {parent=pelvis, type=Null, loc=<-1.0855 -0.2109 8.7053>, rot=<0.0355 -0.0009 0.9993 0.0155>}
 Base_skirt_Collider {parent=pelvis, loc=<-5.9612 7.4861 -4.2372>, rot=<-0.0179 -0.9998 0 0>}
]
variables [
 Parachuteopen {type=bool, default=False}
 isCrouching {type=bool, default=False}
 isEmoting {type=bool, default=False}
 isSkydive {type=bool, default=False}
]
