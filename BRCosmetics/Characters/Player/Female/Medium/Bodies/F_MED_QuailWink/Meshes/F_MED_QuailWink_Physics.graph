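# Physics asset for the F_MED_QuailWink female medium body cosmetic, defining capsule bodies for a left charm and a side chain.
# Pelvis, spine, thigh and skirt collider capsules form the body volume, with spine and thighs locked to the pelvis.
# The three-link charm has limited swing and twist below a base that is fully free relative to spine_01.
# The two-link side chain swings on one axis from a base locked to spine_01.
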PhysicsAsset /BRCosmetics/Characters/Player/Female/Medium/Bodies/F_MED_QuailWink/Meshes/F_MED_QuailWink_Physics
bodies {
 Base_skirt_Collider [{shapeType=Capsule, length=6.1476, radius=10.3305, shapeName=Base_skirt_Collider_capsule, center=<3.9232 -8.7503 0>, rotation=<rot 0 -90 0>}]
 base_charm_01_l [{shapeType=Capsule, length=0.505, radius=0.505, shapeName=base_charm_01_l_capsule, rotation=<rot 0 -90 0>}]
 base_side_chain_01 [{shapeType=Capsule, length=0.505, radius=0.505, shapeName=base_side_chain_01_capsule, rotation=<rot 0 -90 0>}]
 dyn_charm_01_l [{
   shapeType=Capsule
   length=2.1775
   radius=1.0308
   shapeName=dyn_charm_01_l_capsule
   center=<2.647 -0.0055 0.1165>
   rotation=<rot 0 -92.3548 -86.8202>
  }]
 dyn_charm_02_l [{
   shapeType=Capsule
   length=4.6429
   radius=2.5607
   shapeName=dyn_charm_02_l_capsule
   center=<0.9701 -0.1375 1.3448>
   rotation=<rot 0 -178.7348 -82.7943>
  }]
 dyn_charm_03_l [{
   shapeType=Capsule
   length=7.8741
   radius=3.6282
   shapeName=dyn_charm_03_l_capsule
   center=<7.1943 0.1893 1.2514>
   rotation=<rot 0 -88.5134 -86.6831>
  }]
 dyn_side_chain_01 [{shapeType=Capsule, length=8.6726, radius=2.116, shapeName=dyn_side_chain_01_capsule, center=<4.3709 0 2.3317>, rotation=<rot 0 0 90>}]
 dyn_side_chain_02 [{shapeType=Capsule, length=8.115, radius=2.0163, shapeName=dyn_side_chain_02_capsule, center=<6.4498 1.4659 1.4428>, rotation=<rot 0 0 90>}]
 pelvis [{shapeType=Capsule, length=9.1731, radius=8.5243, shapeName=pelvis_capsule, center=<0 -1.5417 0.6452>}]
 spine_01 [{shapeType=Capsule, length=7.0308, radius=5.05, shapeName=spine_01_capsule, center=<0 1.2017 -1.898>, rotation=<rot 0 -90 0>}]
 thigh_l [{shapeType=Capsule, length=18.4206, radius=6.6337, shapeName=thigh_l_capsule, center=<-9.3114 0 0.6304>, rotation=<rot 90 180 0>}]
 thigh_r [{shapeType=Capsule, length=18.4206, radius=6.6337, shapeName=thigh_r_capsule, center=<9.3114 0 -2.0175>, rotation=<rot 90 0 0>}]
}
constraints [
 {
  bone1Name=thigh_r
  bone2Name=pelvis
  swing1LimitDegrees=45
  swing1Motion=Locked
  swing2LimitDegrees=45
  swing2Motion=Locked
  twistLimitDegrees=45
  twistMotion=Locked
 }
 {
  bone1Name=spine_01
  bone2Name=pelvis
  swing1LimitDegrees=45
  swing1Motion=Locked
  swing2LimitDegrees=45
  swing2Motion=Locked
  twistLimitDegrees=45
  twistMotion=Locked
 }
 {
  bone1Name=base_charm_01_l
  bone2Name=spine_01
  swing1LimitDegrees=45
  swing1Motion=Free
  swing2LimitDegrees=45
  swing2Motion=Free
  twistLimitDegrees=45
  twistMotion=Free
 }
 {bone1Name=dyn_charm_01_l, bone2Name=base_charm_01_l, swing1LimitDegrees=20, swing2LimitDegrees=30, twistLimitDegrees=10}
 {bone1Name=dyn_charm_02_l, bone2Name=dyn_charm_01_l, swing1LimitDegrees=25, swing2LimitDegrees=30, twistLimitDegrees=10}
 {bone1Name=dyn_charm_03_l, bone2Name=dyn_charm_02_l, swing1LimitDegrees=25, swing2LimitDegrees=30, twistLimitDegrees=15}
 {
  bone1Name=thigh_l
  bone2Name=pelvis
  swing1LimitDegrees=45
  swing1Motion=Locked
  swing2LimitDegrees=45
  swing2Motion=Locked
  twistLimitDegrees=45
  twistMotion=Locked
 }
 {bone1Name=dyn_side_chain_02, bone2Name=dyn_side_chain_01, swing1LimitDegrees=0, swing2LimitDegrees=20, twistLimitDegrees=0}
 {
  bone1Name=base_side_chain_01
  bone2Name=spine_01
  swing1LimitDegrees=45
  swing1Motion=Locked
  swing2LimitDegrees=45
  swing2Motion=Locked
  twistLimitDegrees=45
  twistMotion=Locked
 }
 {bone1Name=dyn_side_chain_01, bone2Name=base_side_chain_01, swing1LimitDegrees=0, swing2LimitDegrees=20, twistLimitDegrees=0}
]
