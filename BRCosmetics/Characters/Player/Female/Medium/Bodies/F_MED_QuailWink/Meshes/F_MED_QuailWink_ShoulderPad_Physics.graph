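# Physics asset for the right shoulder pad chains on the QuailWink female medium character body.
# Twenty small base capsules lock angularly to upperarm_twist_01_r, each anchoring a dynamic capsule with narrow swing and zero-degree twist limits.
# Strands 08–13 constrain their _01 bones to the bases, leaving their plain dyn bodies unconstrained; strand 09 adds a second link.
# Larger capsules cover spine_05, clavicle and upper arm, with free rotation between upperarm and twist bone.
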PhysicsAsset /BRCosmetics/Characters/Player/Female/Medium/Bodies/F_MED_QuailWink/Meshes/F_MED_QuailWink_ShoulderPad_Physics
bodies {
 base_padchains_01_r [{shapeType=Capsule, length=0.505, radius=0.505, shapeName=base_padchains_01_r_capsule, rotation=<rot 0 -90 0>}]
 base_padchains_02_r [{shapeType=Capsule, length=0.505, radius=0.505, shapeName=base_padchains_02_r_capsule, rotation=<rot 0 -90 0>}]
 base_padchains_03_r [{shapeType=Capsule, length=0.505, radius=0.505, shapeName=base_padchains_03_r_capsule, rotation=<rot 0 -90 0>}]
 base_padchains_04_r [{shapeType=Capsule, length=0.505, radius=0.505, shapeName=base_padchains_04_r_capsule, rotation=<rot 0 -90 0>}]
 base_padchains_05_r [{shapeType=Capsule, length=0.505, radius=0.505, shapeName=base_padchains_05_r_capsule, rotation=<rot 0 -90 0>}]
 base_padchains_06_r [{shapeType=Capsule, length=0.505, radius=0.505, shapeName=base_padchains_06_r_capsule, rotation=<rot 0 -90 0>}]
 base_padchains_07_r [{shapeType=Capsule, length=0.505, radius=0.505, shapeName=base_padchains_07_r_capsule, rotation=<rot 0 -90 0>}]
 base_padchains_08_r [{shapeType=Capsule, length=0.505, radius=0.505, shapeName=base_padchains_08_r_capsule, rotation=<rot 0 -90 0>}]
 base_padchains_09_r [{shapeType=Capsule, length=0.505, radius=0.505, shapeName=base_padchains_09_r_capsule, rotation=<rot 0 -90 0>}]
 base_padchains_10_r [{shapeType=Capsule, length=0.505, radius=0.505, shapeName=base_padchains_10_r_capsule, rotation=<rot 0 -90 0>}]
 base_padchains_11_r [{shapeType=Capsule, length=0.505, radius=0.505, shapeName=base_padchains_11_r_capsule, rotation=<rot 0 -90 0>}]
 base_padchains_12_r [{shapeType=Capsule, length=0.505, radius=0.505, shapeName=base_padchains_12_r_capsule, rotation=<rot 0 -90 0>}]
 base_padchains_13_r [{shapeType=Capsule, length=0.505, radius=0.505, shapeName=base_padchains_13_r_capsule, rotation=<rot 0 -90 0>}]
 base_padchains_14_r [{shapeType=Capsule, length=0.505, radius=0.505, shapeName=base_padchains_14_r_capsule, rotation=<rot 0 -90 0>}]
 base_padchains_15_r [{shapeType=Capsule, length=0.505, radius=0.505, shapeName=base_padchains_15_r_capsule, rotation=<rot 0 -90 0>}]
 base_padchains_16_r [{shapeType=Capsule, length=0.505, radius=0.505, shapeName=base_padchains_16_r_capsule, rotation=<rot 0 -90 0>}]
 base_padchains_17_r [{shapeType=Capsule, length=0.505, radius=0.505, shapeName=base_padchains_17_r_capsule, rotation=<rot 0 -90 0>}]
 base_padchains_18_r [{shapeType=Capsule, length=0.505, radius=0.505, shapeName=base_padchains_18_r_capsule, rotation=<rot 0 -90 0>}]
 base_padchains_19_r [{shapeType=Capsule, length=0.505, radius=0.505, shapeName=base_padchains_19_r_capsule, rotation=<rot 0 -90 0>}]
 base_padchains_20_r [{shapeType=Capsule, length=0.505, radius=0.505, shapeName=base_padchains_20_r_capsule, rotation=<rot 0 -90 0>}]
 clavicle_r [{shapeType=Capsule, length=9.1115, radius=4.7181, shapeName=clavicle_r_capsule, center=<-6.8653 0.8677 4.0063>, rotation=<rot 90 0 0>}]
 dyn_padchains_01_r [{
   shapeType=Capsule
   length=2.1592
   radius=1.0148
   shapeName=dyn_padchains_01_r_capsule
   center=<4.354 0.0129 -0.3526>
   rotation=<rot 85.8085 2.4238 0>
  }]
 dyn_padchains_02_r [{
   shapeType=Capsule
   length=4.3151
   radius=0.5
   shapeName=dyn_padchains_01_r_capsule
   center=<5.0291 -0.0439 0.0099>
   rotation=<rot 86.0117 3.9257 -2.4232>
  }]
 dyn_padchains_03_r [{
   shapeType=Capsule
   length=3.841
   radius=0.5
   shapeName=dyn_padchains_01_r_capsule
   center=<5.3129 0 0.1068>
   rotation=<rot 73.2791 0 -2.4237>
  }]
 dyn_padchains_04_r [{
   shapeType=Capsule
   length=3.5058
   radius=0.5
   shapeName=dyn_padchains_01_r_capsule
   center=<5.6253 0 -0.3169>
   rotation=<rot 82.2757 0.0001 -2.4236>
  }]
 dyn_padchains_05_r [{
   shapeType=Capsule
   length=4.6883
   radius=0.5
   shapeName=dyn_padchains_01_r_capsule
   center=<5.6683 0 0>
   rotation=<rot 83.0602 -0.0001 -2.4238>
  }]
 dyn_padchains_06_r [{shapeType=Capsule, length=5.6521, radius=0.5, shapeName=dyn_padchains_01_r_capsule, center=<7.1804 0 -0.5391>, rotation=<rot 80.6864 0 0>}]
 dyn_padchains_07_r [{
   shapeType=Capsule
   length=4.8603
   radius=1.1224
   shapeName=dyn_padchains_01_r_capsule
   center=<7.3912 -0.4949 0.0481>
   rotation=<rot 80.6023 0.0001 0.0001>
  }]
 dyn_padchains_08_01_r [{
   shapeType=Capsule
   length=6.9486
   radius=0.7039
   shapeName=dyn_padchains_08_01_r_capsule
   center=<8.3848 -0.5138 0.1164>
   rotation=<rot 79.6327 -7.1516 -5.3308>
  }]
 dyn_padchains_08_r [{shapeType=Capsule, length=7.736, radius=0.5, shapeName=dyn_padchains_01_r_capsule, center=<7.5798 -0.4719 0>, rotation=<rot 90 -4.4126 0>}]
 dyn_padchains_09_01_r [{shapeType=Capsule, length=5.1553, radius=0.5, shapeName=dyn_padchains_09_01_r_capsule, center=<3.1603 0 0.3099>, rotation=<rot 90 0 0>}]
 dyn_padchains_09_02_r [{
   shapeType=Capsule
   length=5.1836
   radius=0.5
   shapeName=dyn_padchains_09_02_r_capsule
   center=<3.1664 0.4651 0.5655>
   rotation=<rot 0 -89.4174 -88.9741>
  }]
 dyn_padchains_09_r [{shapeType=Capsule, length=7.1233, radius=0.5, shapeName=dyn_padchains_01_r_capsule, center=<8.3595 -0.0854 0>, rotation=<rot 90 0 0>}]
 dyn_padchains_10_01_r [{shapeType=Capsule, length=3.1992, radius=0.5, shapeName=dyn_padchains_10_01_r_capsule, center=<9.3971 0 0>, rotation=<rot 90 0 0>}]
 dyn_padchains_10_r [{
   shapeType=Capsule
   length=8.5951
   radius=0.5
   shapeName=dyn_padchains_01_r_capsule
   center=<7.9581 0 -0.1575>
   rotation=<rot 88.5554 0.001 0.001>
  }]
 dyn_padchains_11_01_r [{shapeType=Capsule, length=3.1992, radius=0.5, shapeName=dyn_padchains_11_01_r_capsule, center=<9.3971 0 0>, rotation=<rot 90 0 0>}]
 dyn_padchains_11_r [{shapeType=Capsule, length=6.1823, radius=0.5, shapeName=dyn_padchains_01_r_capsule, center=<8.6622 0 0.1803>, rotation=<rot 90 0 0>}]
 dyn_padchains_12_01_r [{shapeType=Capsule, length=3.1992, radius=0.5, shapeName=dyn_padchains_12_01_r_capsule, center=<9.3971 0 0>, rotation=<rot 90 0 0>}]
 dyn_padchains_12_r [{
   shapeType=Capsule
   length=6.6447
   radius=0.5
   shapeName=dyn_padchains_01_r_capsule
   center=<8.2419 -0.0628 0>
   rotation=<rot 85.5534 -2.6893 0.0005>
  }]
 dyn_padchains_13_01_r [{shapeType=Capsule, length=3.1992, radius=0.5, shapeName=dyn_padchains_13_01_r_capsule, center=<9.3971 0 0>, rotation=<rot 90 0 0>}]
 dyn_padchains_13_r [{
   shapeType=Capsule
   length=7.4065
   radius=0.5
   shapeName=dyn_padchains_01_r_capsule
   center=<7.9333 -0.2815 -0.1092>
   rotation=<rot 78.477 0.0002 0.0003>
  }]
 dyn_padchains_14_r [{shapeType=Capsule, length=3.1992, radius=0.5, shapeName=dyn_padchains_14_r_capsule, center=<9.3971 0 0>, rotation=<rot 90 0 0>}]
 dyn_padchains_15_r [{shapeType=Capsule, length=7.3319, radius=0.5, shapeName=dyn_padchains_01_r_capsule, center=<6.4847 -0.2345 0>, rotation=<rot 82.5125 0 0>}]
 dyn_padchains_16_r [{
   shapeType=Capsule
   length=5.6152
   radius=1.4564
   shapeName=dyn_padchains_01_r_capsule
   center=<6.2432 -0.2546 0.047>
   rotation=<rot 84.8296 0 2.1373>
  }]
 dyn_padchains_17_r [{
   shapeType=Capsule
   length=6.3127
   radius=1.098
   shapeName=dyn_padchains_01_r_capsule
   center=<5.3747 0.1965 -0.2677>
   rotation=<rot 86.2422 0 0>
  }]
 dyn_padchains_18_r [{shapeType=Capsule, length=4.1428, radius=0.5, shapeName=dyn_padchains_01_r_capsule, center=<5.6683 0 0.4427>, rotation=<rot 90 0 0>}]
 dyn_padchains_19_r [{shapeType=Capsule, length=4.4333, radius=0.5, shapeName=dyn_padchains_01_r_capsule, center=<5.1056 0 0>, rotation=<rot 90 0 0>}]
 dyn_padchains_20_r [{shapeType=Capsule, length=6.2692, radius=1.1835, shapeName=dyn_padchains_01_r_capsule, center=<5.5464 -0.0492 0.2976>, rotation=<rot 90 0 0>}]
 spine_05 [{shapeType=Capsule, length=5.9356, radius=8.79, center=<0.262 0.5724 0.0091>, rotation=<rot 0.0518 -90 0.1178>}]
 upperarm_r [{shapeType=Capsule, length=16.4771, radius=3.7081, center=<-18.6401 0.5289 0.9302>, rotation=<rot 90 0 0>}]
 upperarm_twist_01_r [
  {shapeType=Capsule, length=9.6836, radius=4.1593, shapeName=upperarm_r_capsule, center=<0.9516 -0.0446 1.5106>, rotation=<rot 90 -1.5228 0>}
  {shapeType=Capsule, length=7.0702, radius=4.4345, shapeName=upperarm_r_capsule, center=<1.9136 -0.5314 4.4604>, rotation=<rot 90 -1.5228 0>}
 ]
}
constraints [
 {bone1Name=upperarm_r, bone2Name=clavicle_r, swing1LimitDegrees=45, swing2LimitDegrees=45, twistLimitDegrees=45}
 {bone1Name=dyn_padchains_20_r, bone2Name=base_padchains_20_r, swing1LimitDegrees=40, swing2LimitDegrees=35, twistLimitDegrees=0}
 {
  bone1Name=base_padchains_01_r
  bone2Name=upperarm_twist_01_r
  swing1LimitDegrees=45
  swing1Motion=Locked
  swing2LimitDegrees=45
  swing2Motion=Locked
  twistLimitDegrees=45
  twistMotion=Locked
 }
 {bone1Name=dyn_padchains_01_r, bone2Name=base_padchains_01_r, swing1LimitDegrees=25, swing2LimitDegrees=20, twistLimitDegrees=0}
 {
  bone1Name=base_padchains_02_r
  bone2Name=upperarm_twist_01_r
  swing1LimitDegrees=45
  swing1Motion=Locked
  swing2LimitDegrees=45
  swing2Motion=Locked
  twistLimitDegrees=45
  twistMotion=Locked
 }
 {bone1Name=dyn_padchains_02_r, bone2Name=base_padchains_02_r, swing1LimitDegrees=25, swing2LimitDegrees=20, twistLimitDegrees=0}
 {
  bone1Name=base_padchains_03_r
  bone2Name=upperarm_twist_01_r
  swing1LimitDegrees=45
  swing1Motion=Locked
  swing2LimitDegrees=45
  swing2Motion=Locked
  twistLimitDegrees=45
  twistMotion=Locked
 }
 {bone1Name=dyn_padchains_03_r, bone2Name=base_padchains_03_r, swing1LimitDegrees=15, swing2LimitDegrees=20, twistLimitDegrees=0}
 {
  bone1Name=base_padchains_04_r
  bone2Name=upperarm_twist_01_r
  swing1LimitDegrees=45
  swing1Motion=Locked
  swing2LimitDegrees=45
  swing2Motion=Locked
  twistLimitDegrees=45
  twistMotion=Locked
 }
 {bone1Name=dyn_padchains_04_r, bone2Name=base_padchains_04_r, swing1LimitDegrees=15, swing2LimitDegrees=20, twistLimitDegrees=0}
 {
  bone1Name=base_padchains_05_r
  bone2Name=upperarm_twist_01_r
  swing1LimitDegrees=45
  swing1Motion=Locked
  swing2LimitDegrees=45
  swing2Motion=Locked
  twistLimitDegrees=45
  twistMotion=Locked
 }
 {bone1Name=dyn_padchains_05_r, bone2Name=base_padchains_05_r, swing1LimitDegrees=15, swing2LimitDegrees=20, twistLimitDegrees=0}
 {
  bone1Name=base_padchains_06_r
  bone2Name=upperarm_twist_01_r
  swing1LimitDegrees=45
  swing1Motion=Locked
  swing2LimitDegrees=45
  swing2Motion=Locked
  twistLimitDegrees=45
  twistMotion=Locked
 }
 {bone1Name=dyn_padchains_06_r, bone2Name=base_padchains_06_r, swing1LimitDegrees=5, swing2LimitDegrees=20, twistLimitDegrees=0}
 {
  bone1Name=base_padchains_07_r
  bone2Name=upperarm_twist_01_r
  swing1LimitDegrees=45
  swing1Motion=Locked
  swing2LimitDegrees=45
  swing2Motion=Locked
  twistLimitDegrees=45
  twistMotion=Locked
 }
 {bone1Name=dyn_padchains_07_r, bone2Name=base_padchains_07_r, swing1LimitDegrees=25, swing2LimitDegrees=20, twistLimitDegrees=0}
 {
  bone1Name=base_padchains_15_r
  bone2Name=upperarm_twist_01_r
  swing1LimitDegrees=45
  swing1Motion=Locked
  swing2LimitDegrees=45
  swing2Motion=Locked
  twistLimitDegrees=45
  twistMotion=Locked
 }
 {bone1Name=dyn_padchains_15_r, bone2Name=base_padchains_15_r, swing1LimitDegrees=25, swing2LimitDegrees=20, twistLimitDegrees=0}
 {
  bone1Name=base_padchains_16_r
  bone2Name=upperarm_twist_01_r
  swing1LimitDegrees=45
  swing1Motion=Locked
  swing2LimitDegrees=45
  swing2Motion=Locked
  twistLimitDegrees=45
  twistMotion=Locked
 }
 {bone1Name=dyn_padchains_16_r, bone2Name=base_padchains_16_r, swing1LimitDegrees=40, swing2LimitDegrees=35, twistLimitDegrees=0}
 {
  bone1Name=base_padchains_17_r
  bone2Name=upperarm_twist_01_r
  swing1LimitDegrees=45
  swing1Motion=Locked
  swing2LimitDegrees=45
  swing2Motion=Locked
  twistLimitDegrees=45
  twistMotion=Locked
 }
 {bone1Name=dyn_padchains_17_r, bone2Name=base_padchains_17_r, swing1LimitDegrees=40, swing2LimitDegrees=35, twistLimitDegrees=0}
 {
  bone1Name=base_padchains_18_r
  bone2Name=upperarm_twist_01_r
  swing1LimitDegrees=45
  swing1Motion=Locked
  swing2LimitDegrees=45
  swing2Motion=Locked
  twistLimitDegrees=45
  twistMotion=Locked
 }
 {bone1Name=dyn_padchains_18_r, bone2Name=base_padchains_18_r, swing1LimitDegrees=25, swing2LimitDegrees=20, twistLimitDegrees=0}
 {
  bone1Name=base_padchains_19_r
  bone2Name=upperarm_twist_01_r
  swing1LimitDegrees=45
  swing1Motion=Locked
  swing2LimitDegrees=45
  swing2Motion=Locked
  twistLimitDegrees=45
  twistMotion=Locked
 }
 {bone1Name=dyn_padchains_19_r, bone2Name=base_padchains_19_r, swing1LimitDegrees=25, swing2LimitDegrees=20, twistLimitDegrees=0}
 {
  bone1Name=base_padchains_20_r
  bone2Name=upperarm_twist_01_r
  swing1LimitDegrees=45
  swing1Motion=Locked
  swing2LimitDegrees=45
  swing2Motion=Locked
  twistLimitDegrees=45
  twistMotion=Locked
 }
 {
  bone1Name=upperarm_twist_01_r
  bone2Name=upperarm_r
  swing1LimitDegrees=45
  swing1Motion=Free
  swing2LimitDegrees=45
  swing2Motion=Free
  twistLimitDegrees=45
  twistMotion=Free
 }
 {bone1Name=dyn_padchains_08_01_r, bone2Name=base_padchains_08_r, swing1LimitDegrees=15, swing2LimitDegrees=20, twistLimitDegrees=0}
 {bone1Name=dyn_padchains_09_01_r, bone2Name=base_padchains_09_r, swing1LimitDegrees=25, swing2LimitDegrees=20, twistLimitDegrees=0}
 {bone1Name=dyn_padchains_10_01_r, bone2Name=base_padchains_10_r, swing1LimitDegrees=15, swing2LimitDegrees=20, twistLimitDegrees=0}
 {bone1Name=dyn_padchains_11_01_r, bone2Name=base_padchains_11_r, swing1LimitDegrees=15, swing2LimitDegrees=20, twistLimitDegrees=0}
 {bone1Name=dyn_padchains_12_01_r, bone2Name=base_padchains_12_r, swing1LimitDegrees=25, swing2LimitDegrees=20, twistLimitDegrees=0}
 {bone1Name=dyn_padchains_13_01_r, bone2Name=base_padchains_13_r, swing1LimitDegrees=25, swing2LimitDegrees=20, twistLimitDegrees=0}
 {bone1Name=dyn_padchains_14_r, bone2Name=base_padchains_14_r, swing1LimitDegrees=25, swing2LimitDegrees=20, twistLimitDegrees=0}
 {
  bone1Name=base_padchains_14_r
  bone2Name=upperarm_twist_01_r
  swing1LimitDegrees=45
  swing1Motion=Locked
  swing2LimitDegrees=45
  swing2Motion=Locked
  twistLimitDegrees=45
  twistMotion=Locked
 }
 {
  bone1Name=base_padchains_13_r
  bone2Name=upperarm_twist_01_r
  swing1LimitDegrees=45
  swing1Motion=Locked
  swing2LimitDegrees=45
  swing2Motion=Locked
  twistLimitDegrees=45
  twistMotion=Locked
 }
 {
  bone1Name=base_padchains_12_r
  bone2Name=upperarm_twist_01_r
  swing1LimitDegrees=45
  swing1Motion=Locked
  swing2LimitDegrees=45
  swing2Motion=Locked
  twistLimitDegrees=45
  twistMotion=Locked
 }
 {
  bone1Name=base_padchains_11_r
  bone2Name=upperarm_twist_01_r
  swing1LimitDegrees=45
  swing1Motion=Locked
  swing2LimitDegrees=45
  swing2Motion=Locked
  twistLimitDegrees=45
  twistMotion=Locked
 }
 {
  bone1Name=base_padchains_10_r
  bone2Name=upperarm_twist_01_r
  swing1LimitDegrees=45
  swing1Motion=Locked
  swing2LimitDegrees=45
  swing2Motion=Locked
  twistLimitDegrees=45
  twistMotion=Locked
 }
 {
  bone1Name=base_padchains_09_r
  bone2Name=upperarm_twist_01_r
  swing1LimitDegrees=45
  swing1Motion=Locked
  swing2LimitDegrees=45
  swing2Motion=Locked
  twistLimitDegrees=45
  twistMotion=Locked
 }
 {
  bone1Name=base_padchains_08_r
  bone2Name=upperarm_twist_01_r
  swing1LimitDegrees=45
  swing1Motion=Locked
  swing2LimitDegrees=45
  swing2Motion=Locked
  twistLimitDegrees=45
  twistMotion=Locked
 }
 {bone1Name=dyn_padchains_09_02_r, bone2Name=dyn_padchains_09_01_r, swing1LimitDegrees=20, swing2LimitDegrees=20, twistLimitDegrees=0}
]
